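# Translucent unlit master material for Niagara particles that renders an animated glitch effect with soft, decal-like fading against scene depth.
# Emissive blends two desaturatable, tiled textures through a noise mask and tints them with particle color.
# Opacity multiplies sine-pulsed texture and time patterns, optionally remapped, by a radial mask and particle alpha.
# A shading path switch uses depth-distance fade for deferred, substituting a sphere mask on forward and mobile.
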Material /Game/Effects/Niagara/MasterMaterials/ParticleDecal/M_Glitch_Decal
graph {
 properties {
  bUsedWithMeshDeformer=true
  bUsedWithNiagaraSprites=true
  bUsedWithSkeletalMesh=true
  blendMode=BLEND_Translucent
  pixelDepthOffsetMode=PDOM_Legacy
  shadingModel=MSM_Unlit
  usedShadingModels=Unlit
 }
 nodes [
  Constant() {r=1}
  n1 = Panner(Coordinate=Multiply(A=TextureCoordinate() {uTiling=0.5, vTiling=0.5}, B=ScalarParameter() {defaultValue=1, parameterName=UVTiling}))
  n2 = Time()
  n3 = Sine(None=Add(
    A=TextureSample(UVs=@n1) {samplerType=SAMPLERTYPE_LinearColor, texture=/Game/TheOutpost/Textures/AnimatedOutpost_Base046_rgb_XVector_a_NormalizedHierPos_UV_2}.R
    B=Multiply(A=@n2)
   )) {period=1.5}
  Cosine(None=@n3)
  n4 = Multiply(A=TextureCoordinate() {uTiling=0.5, vTiling=0.5}, B=ScalarParameter() {defaultValue=1, parameterName=ColorTexTiling})
  n5 = Desaturation(
   None=TextureSample(UVs=@n4) {texture=/Game/Effects/Fort_Effects/Emotes/Energize_Max/LV4/T_Energize_Color_Noise_blur}.RGB
   Fraction=ScalarParameter() {defaultValue=-1, parameterName=Des02}
  )
  n6 = Panner(Coordinate=@n4, Time=Constant() {r=1})
  n7 = Multiply(A=Desaturation(
    None=TextureSampleParameter2D(UVs=@n6) {parameterName=Color, samplerType=SAMPLERTYPE_Normal, texture=/Game/Packages/Fortnite_UniqueMaterials_1/Textures/FoamNormals}.RGB
    Fraction=ScalarParameter() {parameterName=Desaturation}
   )) {constB=5}
  n8 = TextureSample() {samplerType=SAMPLERTYPE_Grayscale, texture=&LowResBlurredNoise}
  n9 = ParticleColor()
  MaterialFunctionCall(Emissive=Multiply(A=LinearInterpolate(A=@n5, B=@n7, Alpha=@n8.R), B=@n9.RGB)) {materialFunction=&MF_FX_EmissiveAdjustments}
  n10 = Panner(Coordinate=Multiply(A=TextureCoordinate() {uTiling=0.2}, B=ScalarParameter() {defaultValue=1, parameterName=UVTiling}))
  n11 = Add(
   A=Power(Base=TextureSample(UVs=@n10) {samplerType=SAMPLERTYPE_LinearColor, texture=&AnimatedOutpost_Base046_rgb_XVector_a_NormalizedHierPos_UV_2}.B)
   B=Multiply(A=Time()) {constB=2}
  )
  n12 = Multiply(A=TextureCoordinate() {uTiling=0.5, vTiling=0.5}, B=ScalarParameter() {defaultValue=1, parameterName=UVTiling})
  n13 = Add(
   A=TextureSample(UVs=Panner(Coordinate=@n12)) {samplerType=SAMPLERTYPE_LinearColor, texture=/Game/Packages/Fortress_SharedMaterials/Textures/T_Slowfield_GridSquares_01}.R
   B=Multiply(A=Time())
  )
  n14 = Multiply(A=Multiply(A=@n3, B=Sine(None=@n11)), B=OneMinus(None=Sine(None=@n13) {period=0.5}))
  n15 = LinearInterpolate(
   A=@n14
   B=LinearInterpolate(A=ScalarParameter() {parameterName=MaskMin}, B=ScalarParameter() {defaultValue=1, parameterName=MaskMax}, Alpha=@n14)
   Alpha=ScalarParameter() {parameterName=UseCustom_MinMask}
  )
  n16 = Power(
   Base=SphereMask(A=TextureCoordinate(), B=Constant() {r=0.5}) {attenuationRadius=0.5, hardnessPercent=0}
   Exp=ScalarParameter() {defaultValue=2, parameterName=SphereMask_Power}
  )
  n17 = MaterialFunctionCall!(Distance Bias, Distance Divisor=ScalarParameter() {defaultValue=50, parameterName=Distance Divisor}, Scene Depth, Pixel Depth) {materialFunction=/Game/Effects/Fort_Effects/Materials/_Functions/MF_DistanceBehindTranslucency_01}
  n18 = SphereMask(A=TextureCoordinate(), B=Constant2Vector() {g=0.5, r=0.5}) {attenuationRadius=0.5, hardnessPercent=0}
  Multiply(
   A=@n9.A
   B=Multiply(
    A=Multiply(A=@n15, B=@n16)
    B=ShadingPathSwitch(Default=@n17.Normalized Inverted, Deferred=@n17.Normalized Inverted, Forward=@n18, Mobile=@n18)
   )
  )
 ]
}
