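# Translucent unlit Niagara sprite material that draws a pixelated body glitch over the screen behind the particle, apparently for characters.
# Screen-snapped SceneColor, tiled by camera depth, blends with clean SceneColor via a dynamic parameter.
# A custom or scene depth threshold, sampled with sine-stepped UV offsets, masks it, combined with particle alpha and a circular SphereMask.
# Step-panned emissive noise is optional; forward and mobile paths output zero.
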Material /Game/Effects/Niagara/MasterMaterials/ParticleDecal/M_ParticleDecal_Character_Body_Glitch
graph {
 properties {bUsedWithNiagaraSprites=true, blendMode=BLEND_Translucent, pixelDepthOffsetMode=PDOM_Legacy, shadingModel=MSM_Unlit, usedShadingModels=Unlit}
 nodes [
  n1 = Multiply(A=Time() {bOverride_Period=true, period=6000}, B=Constant() {r=7}) {constB=4}
  Ceil(None=@n1)
  n2 = ParticleColor()
  n3 = ScalarParameter() {parameterName=UseParticleColor}
  n4 = SceneColor!(UVs)
  n5 = ScreenPosition()
  n6 = ScalarParameter() {defaultValue=1000, parameterName=Fade Length}
  n7 = ScalarParameter() {defaultValue=10, parameterName=Fade offset}
  n8 = Multiply(
   A=ScalarParameter() {defaultValue=30, parameterName=Tile}
   B=LinearInterpolate(Alpha=MaterialFunctionCall(Fade Length=@n6, Fade Offset=@n7) {materialFunction=&CameraDepthFade}.Result) {constA=1, constB=5}
  )
  n9 = Divide(A=Floor(None=Multiply(A=@n5.ViewportUV, B=@n8) {constB=5}), B=@n8)
  n10 = VectorParameter() {parameterName=Speed, defaultValue=<rgba 1 -0.5 0.5 0>}
  n11 = Time()
  n12 = ComponentMask(None=TextureCoordinate()) {g=true}
  n13 = Sine(None=Add(A=Multiply(A=@n10.R, B=@n11), B=@n12)) {period=0.5}
  n14 = Sine(None=Add(A=@n12, B=Multiply(A=@n11, B=@n10.G))) {period=0.7}
  n15 = Add(A=Ceil(None=Sine(None=@n13)), B=Ceil(None=Sine(None=@n14) {period=0.5}))
  n16 = Add(A=ScreenPosition().ViewportUV, B=@n15)
  n17 = SceneTexture(UVs=@n16) {sceneTextureId=PPI_CustomDepth}
  n18 = SceneTexture(UVs=@n16) {sceneTextureId=PPI_SceneDepth}
  n19 = ScalarParameter() {parameterName=Use SceneDepth}
  n20 = Constant()
  n21 = If(
   A=ComponentMask(None=LinearInterpolate(A=@n17.Color, B=@n18.Color, Alpha=@n19)) {b=true, g=true, r=true}
   B=ScalarParameter() {defaultValue=500, parameterName=MaxDe}
   "A > B"=@n20
   "A == B"=@n20
   "A < B"=Constant() {r=1}
  )
  n22 = LinearInterpolate(
   A=@n4.RGB
   B=LinearInterpolate(A=@n4.RGB, B=SceneColor(UVs=@n9).RGB, Alpha=@n21)
   Alpha=DynamicParameter() {paramNames [Body Glitch lerp, Param2, Param3, Param4]}.Body Glitch lerp
  )
  n23 = Constant()
  n24 = ShadingPathSwitch(Default=@n22, Deferred=@n22, Forward=@n23, Mobile=@n23)
  n25 = Floor(None=@n1)
  n26 = TextureSample(UVs=Panner(Coordinate=TextureCoordinate(), Time=Divide(A=@n25) {constB=5.8}) {speedY=3}) {texture=/Game/Accessories/FORT_Backpacks/Backpack_F_MED_IchorIncisor/FX/_old/T_CrystalShield}
  n27 = Multiply(A=Multiply(A=Power(Base=@n26.RGB) {constExponent=1.8}) {constB=6}, B=@n24)
  n28 = ScalarParameter() {parameterName=UseEmissiveNoise}
  MaterialFunctionCall(Emissive=Multiply(A=LinearInterpolate(B=@n2.RGB, Alpha=@n3) {constA=1}, B=LinearInterpolate(A=@n24, B=@n27, Alpha=@n28))) {materialFunction=&MF_FX_EmissiveAdjustments}
  n29 = ParticleColor()
  n30 = ScalarParameter() {parameterName=UseParticleColor}
  n31 = Divide(A=ComponentMask(None=SceneTexture() {sceneTextureId=PPI_SceneDepth}.Color) {b=true, g=true, r=true}) {constB=3000}
  n32 = ScalarParameter() {defaultValue=0.15, parameterName=FECharacterMask}
  n33 = Constant()
  n34 = Constant() {r=1}
  n35 = Multiply(
   A=@n21
   B=LinearInterpolate(B=If(A=@n31, B=@n32, "A > B"=@n33, "A == B"=@n33, "A < B"=@n34), Alpha=ScalarParameter() {parameterName=Use SceneDepth}) {constA=1}
  )
  n36 = Constant()
  Multiply(
   A=Multiply(A=LinearInterpolate(B=@n29.A, Alpha=@n30) {constA=1}, B=ShadingPathSwitch(Default=@n35, Deferred=@n35, Forward=@n36, Mobile=@n36))
   B=SphereMask(A=TextureCoordinate(), B=Constant() {r=0.5}, Hardness=ScalarParameter() {defaultValue=50, parameterName=Hardness}) {attenuationRadius=0.5, hardnessPercent=50}
  )
  Multiply()
  Multiply(A=@n15) {constB=5}
  Saturate(None=@n15)
 ]
}
