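# Translucent unlit Niagara sprite material that draws a glitching, horizontally offset outline around depth silhouettes, disabled on forward and mobile paths.
# Opacity subtracts a depth mask from its sine-jittered copy, scaled by a dynamic parameter and inverted near-camera fade.
# A parameter blends between a pixel-depth range test on custom or scene depth and a sphere-masked scene-depth threshold.
# Emissive multiplies a pulsing-scale, saturation-boosted texture by scene color and particle color.
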Material /Game/Effects/Niagara/MasterMaterials/ParticleDecal/M_ParticleDecal_Outline_Glitch
graph {
 properties {bUsedWithNiagaraSprites=true, blendMode=BLEND_Translucent, pixelDepthOffsetMode=PDOM_Legacy, shadingModel=MSM_Unlit, usedShadingModels=Unlit}
 nodes [
  n1 = ScalarParameter() {defaultValue=15, parameterName=Speed}
  n2 = MaterialFunctionCall(
   UVs=TextureCoordinate()
   Texture Scale=LinearInterpolate(Alpha=MaterialFunctionCall(Speed=@n1) {materialFunction=&DebugTimeSine}.Result) {constA=1, constB=0.8}
  ) {materialFunction=&ScaleUVsByCenter}
  n3 = ParticleColor()
  n4 = Multiply(
   A=Desaturation(
    None=TextureSample(UVs=@n2.UVs) {samplerType=SAMPLERTYPE_LinearColor, texture=/Game/TheOutpost/Textures/AnimatedOutpost_Base046_rgb_XVector_a_NormalizedHierPos_UV_2}.RGB
    Fraction=Constant() {r=-1}
   )
   B=Multiply(A=SceneColor!(UVs).RGB, B=@n3.RGB)
  ) {constA=1}
  n5 = Constant()
  MaterialFunctionCall(Emissive=ShadingPathSwitch(Default=@n4, Deferred=@n4, Forward=@n5, Mobile=@n5)) {materialFunction=&MF_FX_EmissiveAdjustments}
  n6 = LinearInterpolate(Alpha=MaterialFunctionCall(Speed=ScalarParameter() {defaultValue=15, parameterName=Speed}) {materialFunction=&DebugTimeSine}.Result) {constA=-0.007, constB=0.007}
  n7 = MaterialFunctionCall(Distance=ScalarParameter() {defaultValue=500, parameterName="Distance (S)"}) {materialFunction=&Near_Camera_Fade}
  n8 = Multiply(
   A=Multiply(
    A=ScalarParameter() {defaultValue=3, parameterName=offset intensity}
    B=DynamicParameter() {paramNames [OutLine Glitch offset, Param2, Param3, Param4]}.OutLine Glitch offset
   )
   B=Saturate(None=OneMinus(None=@n7.Result))
  )
  n9 = Add(A=ScreenPosition().ViewportUV, B=AppendVector(A=Multiply(A=@n6, B=@n8), B=Constant()))
  n10 = ComponentMask(None=Divide(A=SceneTexture(UVs=@n9) {sceneTextureId=PPI_SceneDepth}.Color) {constB=3000}) {r=true}
  n11 = Constant()
  n12 = ComponentMask(None=Divide(A=SceneTexture() {sceneTextureId=PPI_SceneDepth}.Color) {constB=3000}) {r=true}
  n13 = Constant()
  n14 = Subtract(
   A=If(A=@n10, B=ScalarParameter() {defaultValue=0.15, parameterName=FECharacterMask}, "A > B"=@n11, "A == B"=@n11, "A < B"=Constant() {r=1})
   B=If(A=@n12, B=ScalarParameter() {defaultValue=0.15, parameterName=FECharacterMask}, "A > B"=@n13, "A == B"=@n13, "A < B"=Constant() {r=1})
  )
  n15 = SphereMask(A=TextureCoordinate(), B=Constant() {r=0.5}, Hardness=ScalarParameter() {defaultValue=70, parameterName=Hardness}) {attenuationRadius=0.5, hardnessPercent=70}
  n16 = LinearInterpolate(Alpha=MaterialFunctionCall(Speed=ScalarParameter() {defaultValue=15, parameterName=Speed}) {materialFunction=&DebugTimeSine}.Result) {constA=-0.007, constB=0.007}
  n17 = MaterialFunctionCall(Distance=ScalarParameter() {defaultValue=500, parameterName="Distance (S)"}) {materialFunction=&Near_Camera_Fade}
  n18 = Multiply(
   A=Multiply(
    A=ScalarParameter() {defaultValue=3, parameterName=offset intensity}
    B=DynamicParameter() {paramNames [OutLine Glitch offset, Param2, Param3, Param4]}.OutLine Glitch offset
   )
   B=Saturate(None=OneMinus(None=@n17.Result))
  )
  n19 = Add(A=ScreenPosition().ViewportUV, B=AppendVector(A=Multiply(A=@n16, B=@n18), B=Constant()))
  n20 = LinearInterpolate(
   A=SceneTexture(UVs=@n19) {sceneTextureId=PPI_CustomDepth}.Color
   B=SceneTexture(UVs=@n19) {sceneTextureId=PPI_SceneDepth}.Color
   Alpha=ScalarParameter() {parameterName=Use scene depth}
  )
  n21 = PixelDepth()
  n22 = ScalarParameter() {defaultValue=100, parameterName=range}
  n23 = Constant()
  n24 = If(A=Subtract(A=ComponentMask(None=@n20) {r=true}, B=@n21.R), B=@n22, "A > B"=@n23, "A == B"=@n23, "A < B"=Constant() {r=1})
  n25 = LinearInterpolate(
   A=SceneTexture() {sceneTextureId=PPI_CustomDepth}.Color
   B=SceneTexture() {sceneTextureId=PPI_SceneDepth}.Color
   Alpha=ScalarParameter() {parameterName=Use scene depth}
  )
  n26 = Constant()
  n27 = If(A=Subtract(A=ComponentMask(None=@n25) {r=true}, B=@n21.R), B=@n22, "A > B"=@n26, "A == B"=@n26, "A < B"=Constant() {r=1})
  n28 = Multiply(
   A=@n3.A
   B=LinearInterpolate(A=Subtract(A=@n24, B=@n27), B=Multiply(A=@n14, B=@n15), Alpha=ScalarParameter() {parameterName=UseNewFE_Effact})
  )
  n29 = Constant()
  Reroute(None=ShadingPathSwitch(Default=@n28, Deferred=@n28, Forward=@n29, Mobile=@n29))
  Subtract(A=@n10, B=PixelDepth().R)
  Subtract(A=@n12, B=PixelDepth().R)
 ]
}
